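# Physics asset for the PongoQuell female medium player body, simulating its dangling ribbon and front and back skirt panels.
# Capsules on pelvis, legs and feet, linked by free-rotation joints, form the lower-body shapes alongside the cloth bones.
# The ribbon and each left/right skirt panel hang as limited-angle chains from MSTR root bodies, with boxes for front panels and capsules elsewhere.
# Cross constraints tie lower left and right skirt segments together, mostly with free rotation.
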PhysicsAsset /BRCosmetics/Characters/Player/Female/Medium/Bodies/F_MED_PongoQuell/Meshes/F_MED_PongoQuell_Physics
bodies {
 ball_l [{shapeType=Capsule, length=7, radius=7, shapeName=ball_l_capsule, rotation=<rot -90 -0.0013 0>}]
 ball_r [{shapeType=Capsule, length=7, radius=7, shapeName=ball_r_capsule, rotation=<rot -90 -0.0013 0>}]
 calf_l [{shapeType=Capsule, length=28, radius=7.5, shapeName=calf_l_capsule, center=<-20 -3 0>, rotation=<rot -90 0 0>}]
 calf_r [{shapeType=Capsule, length=28, radius=7.5, shapeName=calf_r_capsule, center=<20 3 0>, rotation=<rot -90 -0.0002 0>}]
 dyn_ribbon_01 [{shapeType=Capsule, length=13, radius=6, shapeName=dyn_ribbon_01_capsule, center=<10.3557 0 7>, rotation=<rot -90 0 0>}]
 dyn_ribbon_02 [{shapeType=Capsule, length=13, radius=6, shapeName=dyn_ribbon_02_capsule, center=<10.3557 0 7>, rotation=<rot -90 0 0>}]
 dyn_ribbon_03 [{shapeType=Capsule, length=13.5, radius=5.5, shapeName=dyn_ribbon_03_capsule, center=<10.3557 0 7>, rotation=<rot -90 0 0>}]
 dyn_ribbon_04 [{shapeType=Capsule, length=13.5, radius=5.5, shapeName=dyn_ribbon_04_capsule, center=<10.3557 0 7>, rotation=<rot -90 0 0>}]
 dyn_ribbon_MSTR [{shapeType=Capsule, length=0.505, radius=0.505, shapeName=dyn_ribbon_MSTR_capsule, rotation=<rot 0 -90 0>}]
 dyn_skirt_bk_01_L [{shapeType=Capsule, length=16.3359, radius=2.6641, shapeName=dyn_skirt_bk_01_L_capsule, center=<8 0 4>, rotation=<rot -90 0 0>}]
 dyn_skirt_bk_01_R [{shapeType=Capsule, length=16.3359, radius=2.6641, shapeName=dyn_skirt_bk_01_R_capsule, center=<8 0 4>, rotation=<rot -90 0 0>}]
 dyn_skirt_bk_02_L [{shapeType=Capsule, length=8.5, radius=6.5, shapeName=dyn_skirt_bk_02_L_capsule, center=<8 0 2>, rotation=<rot -90 0 0>}]
 dyn_skirt_bk_02_R [{shapeType=Capsule, length=8.5, radius=6.5, shapeName=dyn_skirt_bk_02_R_capsule, center=<8 0 2>, rotation=<rot -90 0 0>}]
 dyn_skirt_bk_03_L [{shapeType=Capsule, length=11, radius=9, shapeName=dyn_skirt_bk_03_L_capsule, center=<8 0 2>, rotation=<rot -90 0 0>}]
 dyn_skirt_bk_03_R [{shapeType=Capsule, length=11, radius=9, shapeName=dyn_skirt_bk_03_R_capsule, center=<8 0 2>, rotation=<rot -90 0 0>}]
 dyn_skirt_bk_MSTR_L [{shapeType=Capsule, length=0, radius=0, shapeName=dyn_skirt_bk_MSTR_L_capsule, rotation=<rot -90 0 0>}]
 dyn_skirt_bk_MSTR_R [{shapeType=Capsule, length=0, radius=0, shapeName=dyn_skirt_bk_MSTR_R_capsule, rotation=<rot -90 0 0>}]
 dyn_skirt_fr_01_L [{shapeType=Box, extentX=19, extentY=18, extentZ=1.5, shapeName=dyn_skirt_fr_01_L_box, center=<8.2178 1.5155 3>}]
 dyn_skirt_fr_01_R [{shapeType=Box, extentX=19, extentY=18, extentZ=1.5, shapeName=dyn_skirt_fr_01_L_box, center=<8.2178 -2.1453 3>}]
 dyn_skirt_fr_02_L [{shapeType=Box, extentX=25, extentY=18, extentZ=4.5, shapeName=dyn_skirt_fr_02_L_box, center=<8.2178 1.5155 2>}]
 dyn_skirt_fr_02_R [{shapeType=Box, extentX=25, extentY=18, extentZ=4.5, shapeName=dyn_skirt_fr_02_L_box, center=<8.2178 -2.1453 2>}]
 dyn_skirt_fr_03_L [{shapeType=Box, extentX=22, extentY=18, extentZ=5.5, shapeName=dyn_skirt_fr_03_L_box, center=<8.2178 1.5155 2>}]
 dyn_skirt_fr_03_R [{shapeType=Box, extentX=22, extentY=18, extentZ=5.5, shapeName=dyn_skirt_fr_03_L_box, center=<8.2178 -2.1453 2>}]
 dyn_skirt_fr_MSTR_L [{shapeType=Capsule, length=0, radius=0, shapeName=dyn_skirt_fr_MSTR_L_capsule, rotation=<rot -90 0 0>}]
 dyn_skirt_fr_MSTR_R [{shapeType=Capsule, length=0, radius=0, shapeName=dyn_skirt_fr_MSTR_R_capsule, rotation=<rot -90 0 0>}]
 foot_l [{shapeType=Capsule, length=7.5, radius=7, shapeName=foot_l_capsule, rotation=<rot -90 9.9985 0>}]
 foot_r [{shapeType=Capsule, length=6, radius=7, shapeName=foot_r_capsule, rotation=<rot -90 9.9984 0>}]
 pelvis [{shapeType=Capsule, length=19.4, radius=8.5, shapeName=pelvis_capsule, center=<-3.5933 -0.7284 0>}]
 thigh_l [
  {shapeType=Capsule, length=34, radius=8, shapeName=thigh_l_capsule, center=<-19 0 0>, rotation=<rot 90 0.8546 0>}
  {shapeType=Capsule, length=30, radius=8, shapeName=thigh_l_capsule, center=<-19.8588 -3.0271 0>, rotation=<rot 90 1.8084 0>}
 ]
 thigh_r [
  {shapeType=Capsule, length=34.18, radius=7.82, shapeName=thigh_r_capsule, center=<19 0 0>, rotation=<rot 90 -179.1454 0>}
  {shapeType=Capsule, length=30, radius=8, shapeName=thigh_r_capsule, center=<19.8588 3.0271 0>, rotation=<rot 90 -178.1916 0>}
 ]
}
constraints [
 {
  bone1Name=thigh_l
  bone2Name=pelvis
  swing1LimitDegrees=45
  swing1Motion=Free
  swing2LimitDegrees=45
  swing2Motion=Free
  twistLimitDegrees=45
  twistMotion=Free
 }
 {
  bone1Name=calf_l
  bone2Name=thigh_l
  swing1LimitDegrees=45
  swing1Motion=Free
  swing2LimitDegrees=45
  swing2Motion=Free
  twistLimitDegrees=45
  twistMotion=Free
 }
 {
  bone1Name=foot_l
  bone2Name=calf_l
  swing1LimitDegrees=45
  swing1Motion=Free
  swing2LimitDegrees=45
  swing2Motion=Free
  twistLimitDegrees=45
  twistMotion=Free
 }
 {
  bone1Name=ball_l
  bone2Name=foot_l
  swing1LimitDegrees=45
  swing1Motion=Free
  swing2LimitDegrees=45
  swing2Motion=Free
  twistLimitDegrees=45
  twistMotion=Free
 }
 {
  bone1Name=thigh_r
  bone2Name=pelvis
  swing1LimitDegrees=45
  swing1Motion=Free
  swing2LimitDegrees=45
  swing2Motion=Free
  twistLimitDegrees=45
  twistMotion=Free
 }
 {
  bone1Name=calf_r
  bone2Name=thigh_r
  swing1LimitDegrees=45
  swing1Motion=Free
  swing2LimitDegrees=45
  swing2Motion=Free
  twistLimitDegrees=45
  twistMotion=Free
 }
 {
  bone1Name=foot_r
  bone2Name=calf_r
  swing1LimitDegrees=45
  swing1Motion=Free
  swing2LimitDegrees=45
  swing2Motion=Free
  twistLimitDegrees=45
  twistMotion=Free
 }
 {
  bone1Name=ball_r
  bone2Name=foot_r
  swing1LimitDegrees=45
  swing1Motion=Free
  swing2LimitDegrees=45
  swing2Motion=Free
  twistLimitDegrees=45
  twistMotion=Free
 }
 {bone1Name=dyn_ribbon_01, bone2Name=dyn_ribbon_MSTR, swing1LimitDegrees=10, swing2LimitDegrees=15, twistLimitDegrees=2.5}
 {bone1Name=dyn_ribbon_02, bone2Name=dyn_ribbon_01, swing1LimitDegrees=10, swing2LimitDegrees=15, twistLimitDegrees=2.5}
 {bone1Name=dyn_ribbon_03, bone2Name=dyn_ribbon_02, swing1LimitDegrees=10, swing2LimitDegrees=15, twistLimitDegrees=2.5}
 {bone1Name=dyn_ribbon_04, bone2Name=dyn_ribbon_03, swing1LimitDegrees=10, swing2LimitDegrees=15, twistLimitDegrees=2.5}
 {bone1Name=dyn_skirt_bk_01_L, bone2Name=dyn_skirt_bk_MSTR_L, swing1LimitDegrees=2.5, swing2LimitDegrees=15, twistLimitDegrees=5}
 {bone1Name=dyn_skirt_bk_02_L, bone2Name=dyn_skirt_bk_01_L, swing1LimitDegrees=10, swing2LimitDegrees=23, twistLimitDegrees=5}
 {bone1Name=dyn_skirt_bk_03_L, bone2Name=dyn_skirt_bk_02_L, swing1LimitDegrees=10, swing2LimitDegrees=28, twistLimitDegrees=5}
 {bone1Name=dyn_skirt_bk_01_R, bone2Name=dyn_skirt_bk_MSTR_R, swing1LimitDegrees=2.5, swing2LimitDegrees=15, twistLimitDegrees=5}
 {bone1Name=dyn_skirt_bk_02_R, bone2Name=dyn_skirt_bk_01_R, swing1LimitDegrees=10, swing2LimitDegrees=23, twistLimitDegrees=5}
 {bone1Name=dyn_skirt_bk_03_R, bone2Name=dyn_skirt_bk_02_R, swing1LimitDegrees=10, swing2LimitDegrees=28, twistLimitDegrees=5}
 {bone1Name=dyn_skirt_fr_01_L, bone2Name=dyn_skirt_fr_MSTR_L, swing1LimitDegrees=10, swing2LimitDegrees=15, twistLimitDegrees=5}
 {bone1Name=dyn_skirt_fr_02_L, bone2Name=dyn_skirt_fr_01_L, swing1LimitDegrees=10, swing2LimitDegrees=20, twistLimitDegrees=5}
 {bone1Name=dyn_skirt_fr_03_L, bone2Name=dyn_skirt_fr_02_L, swing1LimitDegrees=10, swing2LimitDegrees=25, twistLimitDegrees=5}
 {bone1Name=dyn_skirt_fr_01_R, bone2Name=dyn_skirt_fr_MSTR_R, swing1LimitDegrees=10, swing2LimitDegrees=15, twistLimitDegrees=5}
 {bone1Name=dyn_skirt_fr_02_R, bone2Name=dyn_skirt_fr_01_R, swing1LimitDegrees=10, swing2LimitDegrees=20, twistLimitDegrees=5}
 {bone1Name=dyn_skirt_fr_03_R, bone2Name=dyn_skirt_fr_02_R, swing1LimitDegrees=10, swing2LimitDegrees=25, twistLimitDegrees=5}
 {
  bone1Name=dyn_skirt_fr_03_R
  bone2Name=dyn_skirt_fr_03_L
  swing1LimitDegrees=45
  swing1Motion=Free
  swing2LimitDegrees=45
  swing2Motion=Free
  twistLimitDegrees=45
  twistMotion=Free
 }
 {
  bone1Name=dyn_skirt_fr_02_R
  bone2Name=dyn_skirt_fr_02_L
  swing1LimitDegrees=45
  swing1Motion=Free
  swing2LimitDegrees=45
  swing2Motion=Free
  twistLimitDegrees=45
  twistMotion=Free
 }
 {bone1Name=dyn_skirt_bk_03_R, bone2Name=dyn_skirt_bk_03_L, swing1LimitDegrees=20, swing2LimitDegrees=20, twistLimitDegrees=0}
 {
  bone1Name=dyn_skirt_bk_02_R
  bone2Name=dyn_skirt_bk_02_L
  swing1LimitDegrees=45
  swing1Motion=Free
  swing2LimitDegrees=45
  swing2Motion=Free
  twistLimitDegrees=45
  twistMotion=Free
 }
]
