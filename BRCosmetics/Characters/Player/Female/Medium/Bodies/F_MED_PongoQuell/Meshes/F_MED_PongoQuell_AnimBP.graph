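# Animation Blueprint for the F_MED_PongoQuell character body, adding secondary motion to skirt, ribbon and shoelace bones.
# The parent mesh's copied pose runs through Control Rig, shoelace constraints and AnimDynamics into a cached pose.
# A crouch flag, off during baseball slides, picks between crouching and default RigidBody simulations, followed by additive skirt and ribbon rotations.
# The update event derives Control Rig inputs from movement state and speed.
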AnimBlueprint /BRCosmetics/Characters/Player/Female/Medium/Bodies/F_MED_PongoQuell/Meshes/F_MED_PongoQuell_AnimBP
target_skeleton=&/F_MED_PongoQuell_Skeleton
graphs [
 AnimGraph {nodes [
   n1 = RigidBody!(
    ComponentPose=Local To Component(LocalPose=Use cached pose 'start pose'())
    Alpha
    OverrideWorldGravity:Vector="0,0,-980"
    SimSpaceSettings:Sim Space Settings Structure=(WorldAlpha=1,VelocityScaleZ=1,DampingAlpha=1,MaxLinearVelocity=10000,MaxAngularVelocity=10000,MaxLinearAcceleration=10000,MaxAngularAcceleration=10000,ExternalLinearDragV=(X=0,Y=0,Z=0),ExternalLinearVelocity=(X=0,Y=0,Z=0),ExternalAngularVelocity=(X=0,Y=0,Z=0))
   ) -> (Pose) {
    bClampLinearTranslationLimitToRefPose=true
    bOverrideWorldGravity=true
    overridePhysicsAsset=&/F_MED_PongoQuell_Crouching_Physics
    simulationSpace=BaseBoneSpace
    baseBoneRef {boneName=pelvis}
    simSpaceSettings {worldAlpha=1}
   }
   n2 = Component To Local(ComponentPose=@n1)
   n3 = RigidBody(
    ComponentPose=Local To Component(LocalPose=Use cached pose 'start pose'())
    OverrideWorldGravity="0,0,-980"
    SimSpaceSettings=(WorldAlpha=1,VelocityScaleZ=1,DampingAlpha=1,MaxLinearVelocity=10000,MaxAngularVelocity=10000,MaxLinearAcceleration=10000,MaxAngularAcceleration=10000,ExternalLinearDragV=(X=0,Y=0,Z=0),ExternalLinearVelocity=(X=0,Y=0,Z=0),ExternalAngularVelocity=(X=0,Y=0,Z=0))
   ) {
    bClampLinearTranslationLimitToRefPose=true
    bOverrideWorldGravity=true
    overridePhysicsAsset=&/F_MED_PongoQuell_Physics
    simulationSpace=BaseBoneSpace
    baseBoneRef {boneName=pelvis}
    simSpaceSettings {worldAlpha=1}
   }
   n4 = Component To Local(ComponentPose=@n3)
   n5 = Get Crouch_My() -> (Crouch_My:bool)
   n6 = "Transform (Modify) Bone - Bone: dyn_skirt_fr_02_L"(
    ComponentPose=Local To Component(LocalPose=Blend Poses by bool(BlendPose_0=@n2, BlendPose_1=@n4, bActiveValue=@n5))
    Alpha=1
    Translation="0,0,0"
    Rotation="15,0,0"
    Scale="1,1,1"
   ) {rotationMode=BMM_Additive, rotationSpace=BCS_BoneSpace, boneToModify {boneName=dyn_skirt_fr_02_L}, rotation {pitch=15}}
   n7 = "Transform (Modify) Bone - Bone: dyn_skirt_fr_02_R"(ComponentPose=@n6, Alpha=1, Translation="0,0,0", Rotation="15,0,0", Scale="1,1,1") {rotationMode=BMM_Additive, rotationSpace=BCS_BoneSpace, boneToModify {boneName=dyn_skirt_fr_02_R}, rotation {pitch=15}}
   n8 = "Transform (Modify) Bone - Bone: dyn_ribbon_01"(ComponentPose=@n7, Alpha=1, Translation="0,0,0", Rotation="10,0,0", Scale="1,1,1") {rotationMode=BMM_Additive, rotationSpace=BCS_BoneSpace, boneToModify {boneName=dyn_ribbon_01}, rotation {pitch=10}}
   n9 = "Transform (Modify) Bone - Bone: dyn_skirt_bk_01_L"(
    ComponentPose="Transform (Modify) Bone - Bone: dyn_ribbon_03"(
     ComponentPose="Transform (Modify) Bone - Bone: dyn_ribbon_02"(ComponentPose=@n8, Alpha=1, Translation="0,0,0", Rotation="5,0,0", Scale="1,1,1") {rotationMode=BMM_Additive, rotationSpace=BCS_BoneSpace, boneToModify {boneName=dyn_ribbon_02}, rotation {pitch=5}}
     Alpha=1
     Translation="0,0,0"
     Rotation="0,0,0"
     Scale="1,1,1"
    ) {rotationMode=BMM_Additive, rotationSpace=BCS_BoneSpace, boneToModify {boneName=dyn_ribbon_03}}
    Alpha=1
    Translation="0,0,0"
    Rotation="10,0,0"
    Scale="1,1,1"
   ) {rotationMode=BMM_Additive, rotationSpace=BCS_BoneSpace, boneToModify {boneName=dyn_skirt_bk_01_L}, rotation {pitch=10}}
   n10 = "Transform (Modify) Bone - Bone: dyn_skirt_bk_01_R"(ComponentPose=@n9, Alpha=1, Translation="0,0,0", Rotation="10,0,0", Scale="1,1,1") {rotationMode=BMM_Additive, rotationSpace=BCS_BoneSpace, boneToModify {boneName=dyn_skirt_bk_01_R}, rotation {pitch=10}}
   n11 = "Transform (Modify) Bone - Bone: dyn_skirt_bk_02_L"(ComponentPose=@n10, Alpha=1, Translation="0,0,0", Rotation="10,0,0", Scale="1,1,1") {rotationMode=BMM_Additive, rotationSpace=BCS_BoneSpace, boneToModify {boneName=dyn_skirt_bk_02_L}, rotation {pitch=10}}
   n12 = "Transform (Modify) Bone - Bone: dyn_skirt_bk_02_R"(ComponentPose=@n11, Alpha=1, Translation="0,0,0", Rotation="10,0,0", Scale="1,1,1") {rotationMode=BMM_Additive, rotationSpace=BCS_BoneSpace, boneToModify {boneName=dyn_skirt_bk_02_R}, rotation {pitch=10}}
   n13 = "Transform (Modify) Bone - Bone: dyn_skirt_bk_03_L"(ComponentPose=@n12, Alpha=1, Translation="0,0,0", Rotation="10,0,0", Scale="1,1,1") {rotationMode=BMM_Additive, rotationSpace=BCS_BoneSpace, boneToModify {boneName=dyn_skirt_bk_03_L}, rotation {pitch=10}}
   n14 = "Transform (Modify) Bone - Bone: dyn_skirt_bk_03_R"(ComponentPose=@n13, Alpha=1, Translation="0,0,0", Rotation="10,0,0", Scale="1,1,1") {rotationMode=BMM_Additive, rotationSpace=BCS_BoneSpace, boneToModify {boneName=dyn_skirt_bk_03_R}, rotation {pitch=10}}
   n15 = "Transform (Modify) Bone - Bone: dyn_ribbon_01"(ComponentPose=@n14, Alpha=1, Translation="0,0,0", Rotation="10,0,0", Scale="1,1,1") {rotationMode=BMM_Additive, rotationSpace=BCS_BoneSpace, boneToModify {boneName=dyn_ribbon_01}, rotation {pitch=10}}
   n16 = "Transform (Modify) Bone - Bone: dyn_skirt_bk_01_L"(
    ComponentPose="Transform (Modify) Bone - Bone: dyn_ribbon_03"(
     ComponentPose="Transform (Modify) Bone - Bone: dyn_ribbon_02"(ComponentPose=@n15, Alpha=1, Translation="0,0,0", Rotation="0,0,0", Scale="1,1,1") {rotationMode=BMM_Additive, rotationSpace=BCS_BoneSpace, boneToModify {boneName=dyn_ribbon_02}}
     Alpha=1
     Translation="0,0,0"
     Rotation="0,0,0"
     Scale="1,1,1"
    ) {rotationMode=BMM_Additive, rotationSpace=BCS_BoneSpace, boneToModify {boneName=dyn_ribbon_03}}
    Alpha=1
    Translation="0,0,0"
    Rotation="10,0,0"
    Scale="1,1,1"
   ) {rotationMode=BMM_Additive, rotationSpace=BCS_BoneSpace, boneToModify {boneName=dyn_skirt_bk_01_L}, rotation {pitch=10}}
   n17 = "Transform (Modify) Bone - Bone: dyn_skirt_bk_01_R"(ComponentPose=@n16, Alpha=1, Translation="0,0,0", Rotation="10,0,0", Scale="1,1,1") {rotationMode=BMM_Additive, rotationSpace=BCS_BoneSpace, boneToModify {boneName=dyn_skirt_bk_01_R}, rotation {pitch=10}}
   n18 = "Transform (Modify) Bone - Bone: dyn_skirt_bk_02_L"(ComponentPose=@n17, Alpha=1, Translation="0,0,0", Rotation="5,0,0", Scale="1,1,1") {rotationMode=BMM_Additive, rotationSpace=BCS_BoneSpace, boneToModify {boneName=dyn_skirt_bk_02_L}, rotation {pitch=5}}
   n19 = "Transform (Modify) Bone - Bone: dyn_skirt_bk_02_R"(ComponentPose=@n18, Alpha=1, Translation="0,0,0", Rotation="5,0,0", Scale="1,1,1") {rotationMode=BMM_Additive, rotationSpace=BCS_BoneSpace, boneToModify {boneName=dyn_skirt_bk_02_R}, rotation {pitch=5}}
   n20 = "Transform (Modify) Bone - Bone: dyn_skirt_bk_03_L"(ComponentPose=@n19, Alpha=1, Translation="0,0,0", Rotation="5,0,0", Scale="1,1,1") {rotationMode=BMM_Additive, rotationSpace=BCS_BoneSpace, boneToModify {boneName=dyn_skirt_bk_03_L}, rotation {pitch=5}}
   n21 = "Transform (Modify) Bone - Bone: dyn_skirt_bk_03_R"(ComponentPose=@n20, Alpha=1, Translation="0,0,0", Rotation="5,0,0", Scale="1,1,1") {rotationMode=BMM_Additive, rotationSpace=BCS_BoneSpace, boneToModify {boneName=dyn_skirt_bk_03_R}, rotation {pitch=5}}
   n22 = "Transform (Modify) Bone - Bone: dyn_ribbon_02"(
    ComponentPose="Transform (Modify) Bone - Bone: dyn_ribbon_01"(ComponentPose=@n21, Alpha=1, Translation="0,0,0", Rotation="5,0,0", Scale="1,1,1") {rotationMode=BMM_Additive, rotationSpace=BCS_BoneSpace, boneToModify {boneName=dyn_ribbon_01}, rotation {pitch=5}}
    Alpha=1
    Translation="0,0,0"
    Rotation="-2.5,0,0"
    Scale="1,1,1"
   ) {rotationMode=BMM_Additive, rotationSpace=BCS_BoneSpace, boneToModify {boneName=dyn_ribbon_02}, rotation {pitch=-2.5}}
   n23 = "Transform (Modify) Bone - Bone: dyn_ribbon_04"(
    ComponentPose="Transform (Modify) Bone - Bone: dyn_ribbon_03"(ComponentPose=@n22, Alpha=1, Translation="0,0,0", Rotation="0,0,0", Scale="1,1,1") {rotationMode=BMM_Additive, rotationSpace=BCS_BoneSpace, boneToModify {boneName=dyn_ribbon_03}}
    Alpha=1
    Translation="0,0,0"
    Rotation="0,0,0"
    Scale="1,1,1"
   ) {rotationMode=BMM_Additive, rotationSpace=BCS_BoneSpace, boneToModify {boneName=dyn_ribbon_04}}
   n24 = "Transform (Modify) Bone - Bone: dyn_skirt_bk_03_L"(
    ComponentPose="Transform (Modify) Bone - Bone: dyn_skirt_bk_02_R"(
     ComponentPose="Transform (Modify) Bone - Bone: dyn_skirt_bk_02_L"(ComponentPose=@n23, Alpha=1, Translation="0,0,0", Rotation="0,0,0", Scale="1,1,1") {rotationMode=BMM_Additive, rotationSpace=BCS_BoneSpace, boneToModify {boneName=dyn_skirt_bk_02_L}}
     Alpha=1
     Translation="0,0,0"
     Rotation="0,0,0"
     Scale="1,1,1"
    ) {rotationMode=BMM_Additive, rotationSpace=BCS_BoneSpace, boneToModify {boneName=dyn_skirt_bk_02_R}}
    Alpha=1
    Translation="0,0,0"
    Rotation="15,0,0"
    Scale="1,1,1"
   ) {rotationMode=BMM_Additive, rotationSpace=BCS_BoneSpace, boneToModify {boneName=dyn_skirt_bk_03_L}, rotation {pitch=15}}
   n25 = "Transform (Modify) Bone - Bone: dyn_skirt_bk_03_R"(ComponentPose=@n24, Alpha=1, Translation="0,0,0", Rotation="15,0,0", Scale="1,1,1") {rotationMode=BMM_Additive, rotationSpace=BCS_BoneSpace, boneToModify {boneName=dyn_skirt_bk_03_R}, rotation {pitch=15}}
   Output Pose(Result=Component To Local(ComponentPose=@n25)) {layerGroup=None}
   n26 = Control Rig!(
    Source=Copy Pose From Mesh() {bUseAttachedParent=true}
    Alpha
    IsRunning:double=Get IsRunning() -> (IsRunning:double)
    isCrouching:double=Get is Crouching() -> (is Crouching:double)
    is knee sliding:double=Get is knee sliding() -> (is knee sliding:double)
    IsSkydiveFloating:double=Get IsSkydiveFloating() -> (IsSkydiveFloating:double)
   ) -> (Pose) {controlRigAssetReference {controlRigAsset=&/F_MED_PongoQuell_ControlRig}}
   n27 = "Constraint - dyn_shoelace_CST_L (PPP)"!(
    ComponentPose=Local To Component(LocalPose=@n26)
    Alpha=1
    ConstraintWeights_0=0.5965
    ConstraintWeights_1:float=0.3936
    ConstraintWeights_2:float=0.01
   ) -> (Pose) {
    boneToModify {boneName=dyn_shoelace_CST_L}
    constraintSetup [{targetBone {boneName=calf_l}}, {targetBone {boneName=calf_twist_02_l}}, {targetBone {boneName=deform_knee_out_l}}]
    constraintWeights [1,1,1]
   }
   n28 = "Constraint - dyn_shoelace_CST_R (PPP)"!(ComponentPose=@n27, Alpha=1, ConstraintWeights_0=0.5965, ConstraintWeights_1:float=0.3936, ConstraintWeights_2:float=0.01) -> (Pose) {
    boneToModify {boneName=dyn_shoelace_CST_R}
    constraintSetup [{targetBone {boneName=calf_r}}, {targetBone {boneName=calf_twist_02_r}}, {targetBone {boneName=deform_knee_out_r}}]
    constraintWeights [1,1,1]
   }
   n29 = Anim Dynamics - Bone: dyn_shoelace_out_L!(ComponentPose=@n28, Alpha=1, GravityOverride:Vector="0,0,0") -> (Pose) {
    angularDampingOverride=1
    angularSpringConstant=1000
    bAngularSpring=true
    bOverrideAngularDamping=true
    bUseGravityOverride=true
    bUsePlanarLimit=false
    simulationSpace=RootRelative
    boundBone {boneName=dyn_shoelace_out_L}
    componentAppliedLinearAccClamp=<1000 1000 1000>
    componentLinearAccScale=<0.3 0.3 0.3>
    componentLinearVelScale=<0.3 0.3 0.3>
    physicsBodyDefinitions [{
      sphereCollisionRadius=2.5
      boundBone {boneName=dyn_shoelace_out_L}
      boxExtents=<10 10 10>
      constraintSetup {angularLimitsMax=<0 0 15>, angularLimitsMin=<0 -35 -15>, angularTarget=<1 0 0>}
      localJointOffset=<3 0 0>
     }]
   }
   n30 = Anim Dynamics - Bone: dyn_shoelace_in_L!(ComponentPose=@n29, Alpha=1, GravityOverride:Vector="0,0,0") -> (Pose) {
    angularDampingOverride=1
    angularSpringConstant=1000
    bAngularSpring=true
    bOverrideAngularDamping=true
    bUseGravityOverride=true
    bUsePlanarLimit=false
    simulationSpace=RootRelative
    boundBone {boneName=dyn_shoelace_in_L}
    componentAppliedLinearAccClamp=<1000 1000 1000>
    componentLinearAccScale=<0.3 0.3 0.3>
    componentLinearVelScale=<0.3 0.3 0.3>
    physicsBodyDefinitions [{
      sphereCollisionRadius=2.5
      boundBone {boneName=dyn_shoelace_in_L}
      boxExtents=<10 10 10>
      constraintSetup {angularLimitsMax=<0 35 15>, angularLimitsMin=<0 0 -15>, angularTarget=<1 0 0>}
      localJointOffset=<3 0 0>
     }]
   }
   n31 = Anim Dynamics - Bone: dyn_shoelace_in_R!(ComponentPose=@n30, Alpha=1, GravityOverride:Vector="0,0,0") -> (Pose) {
    angularDampingOverride=1
    angularSpringConstant=1000
    bAngularSpring=true
    bOverrideAngularDamping=true
    bUseGravityOverride=true
    bUsePlanarLimit=false
    simulationSpace=RootRelative
    boundBone {boneName=dyn_shoelace_in_R}
    componentAppliedLinearAccClamp=<1000 1000 1000>
    componentLinearAccScale=<0.3 0.3 0.3>
    componentLinearVelScale=<0.3 0.3 0.3>
    physicsBodyDefinitions [{
      sphereCollisionRadius=2.5
      boundBone {boneName=dyn_shoelace_in_R}
      boxExtents=<10 10 10>
      constraintSetup {angularLimitsMax=<0 0 15>, angularLimitsMin=<0 -35 -15>, angularTarget=<1 0 0>}
      localJointOffset=<3 0 0>
     }]
   }
   n32 = Anim Dynamics - Bone: dyn_shoelace_out_R!(ComponentPose=@n31, Alpha=1, GravityOverride:Vector="0,0,0") -> (Pose) {
    angularDampingOverride=1
    angularSpringConstant=1000
    bAngularSpring=true
    bOverrideAngularDamping=true
    bUseGravityOverride=true
    bUsePlanarLimit=false
    simulationSpace=RootRelative
    boundBone {boneName=dyn_shoelace_out_R}
    componentAppliedLinearAccClamp=<1000 1000 1000>
    componentLinearAccScale=<0.3 0.3 0.3>
    componentLinearVelScale=<0.3 0.3 0.3>
    physicsBodyDefinitions [{
      sphereCollisionRadius=2.5
      boundBone {boneName=dyn_shoelace_out_R}
      boxExtents=<10 10 10>
      constraintSetup {angularLimitsMax=<0 35 15>, angularLimitsMin=<0 0 -15>, angularTarget=<1 0 0>}
      localJointOffset=<3 0 0>
     }]
   }
   Save cached pose 'start pose'(Pose=Component To Local(ComponentPose=@n32)) {cacheName=start pose}
  ]}
 EventGraph {nodes [
   n1 = Event BlueprintUpdateAnimation()
   > Set Delta Time X(Delta Time X=@n1.DeltaTimeX)
   > n2 = "Gravity Override State Function (Parameter Asset) (Nativized)"(
    CharacterPartAnimInstance=Self-Reference()
    GameTimeInSeconds=GetGameTimeInSeconds()
    Parameters=&/F_MED_PongoQuell_Parameters
   )
   > Set Gravity Override(Gravity Override=@n2)
   > Set RBAN Settings Out(RBAN Settings Out=@n2.OutRigidBodyAnimNodeParameters)
   > Set is Crouching(is Crouching:double=SelectFloat(A=1, B=0, bPickA=Get bIsCrouching())) -> (Output_Get:double)
   > Set is SkyDiving(is SkyDiving:double=SelectFloat(A=1, B=0, bPickA=Get bIsSkyDiving())) -> (Output_Get:double)
   > Set IsSkydiveFloating(IsSkydiveFloating:double=SelectFloat(A=1, B=0, bPickA=Get bIsSkydiveFloating())) -> (Output_Get:double)
   n3 = TryGetPawnOwner()
   > Set Velocity(Velocity:double=Vector Length XY!(A:Vector=GetVelocity!(self:Actor Object Reference=@n3) -> (ReturnValue:Vector)) -> (ReturnValue:double)) -> (Output_Get:double)
   n4 = FInterpTo(Current=Get IsRunning(), Target=0, DeltaTime=Get Delta Time X(), InterpSpeed=1)
   n5 = Get Velocity() -> (Velocity:double)
   n6 = NOT Boolean(A=OR Boolean(A=Get bIsCrouching(), B=Get bIsSliding()))
   n7 = Get bIsOnGround() -> (bIsOnGround:bool)
   > Set IsRunning(IsRunning:double=SelectFloat(A=MapRangeClamped(Value=@n5, InRangeA=250, InRangeB=550, OutRangeA=@n4), B=0, bPickA=AND Boolean(A=@n6, B=@n7))) -> (Output_Get:double)
   n8 = Get Velocity()
   n9 = NOT Boolean(A=OR Boolean(A=Get bIsCrouching(), B=Get bIsSliding()))
   n10 = OR Boolean!(A=Get bIsJumping(), B=Get bIsOnGround(), C:bool=Get bIsFalling()) -> (ReturnValue)
   > Set is Sprinting(is Sprinting:double=SelectFloat(A=MapRangeClamped(Value=@n8, InRangeA=550, InRangeB=770), B=0, bPickA=AND Boolean(A=@n10, B=@n9))) -> (Output_Get:double)
   n11 = Get Delta Time X()
   n12 = Get is knee sliding()
   n13 = Get Delta Time X()
   n14 = Get is knee sliding()
   > Set is knee sliding(is knee sliding:double=SelectFloat(
     A=FInterpTo(Current=@n12, Target=1, DeltaTime=@n11, InterpSpeed=5)
     B=FInterpTo(Current=@n14, Target=0, DeltaTime=@n13, InterpSpeed=5)
     bPickA=Get bIsKneeSliding()
    )) -> (Output_Get:double)
   n15 = Get bIsBaseballSliding()
   > Set Crouch_My(Crouch_My:bool=AND Boolean(A=Get bIsCrouching(), B=NOT Boolean(A=@n15))) -> (Output_Get:bool)
   OR Boolean!(A=Get bIsJumping(), B=false) -> (ReturnValue)
   Get Velocity()
  ]}
]
variables [
 Crouch_My {type=boolean}
 Delta Time X {type=number}
 Gravity Override {type=Vector}
 is Crouching {type=number}
 is knee sliding {type=number}
 is SkyDiving {type=number}
 is Sprinting {type=number}
 IsRunning {type=number}
 IsSkydiveFloating {type=number}
 RBAN Settings Out {type=FortRigidBodyAnimNodeParameters}
 Velocity {type=number}
]
