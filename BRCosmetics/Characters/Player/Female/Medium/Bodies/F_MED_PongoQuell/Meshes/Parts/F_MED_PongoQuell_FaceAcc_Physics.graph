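# Physics asset for the PongoQuell female character's face accessory mesh, defining capsule bodies and constraints for dynamic hair.
# Head, neck, upper spine, clavicle and upper-arm capsules are linked by free-motion constraints.
# Front, middle and back chains per side hang from head via mostly zero-size master bodies with narrow limits; back chains fork into three ends.
# A tighter extra left chain has no constraint to head.
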PhysicsAsset /BRCosmetics/Characters/Player/Female/Medium/Bodies/F_MED_PongoQuell/Meshes/Parts/F_MED_PongoQuell_FaceAcc_Physics
bodies {
 clavicle_l [{shapeType=Capsule, length=6.5, radius=9, shapeName=clavicle_l_capsule, center=<3.3286 0 -4.5921>, rotation=<rot -70 179.9988 -179.9987>}]
 clavicle_r [{shapeType=Capsule, length=6.5, radius=9, shapeName=clavicle_r_capsule, center=<-3.4353 0 4.2991>, rotation=<rot -70 179.9998 -179.9998>}]
 dyn_hair_01_L [{shapeType=Capsule, length=5.5, radius=0.7, shapeName=dyn_hair_01_L_capsule, center=<0.5 0 0>, rotation=<rot -90 0 0>}]
 dyn_hair_02_L [{shapeType=Capsule, length=5.5, radius=0.5, shapeName=dyn_hair_02_L_capsule, center=<0.5 0 0>, rotation=<rot -90 0 0>}]
 dyn_hair_MSTR_L [{shapeType=Capsule, length=0, radius=0.505, shapeName=dyn_hair_MSTR_L_capsule, rotation=<rot -79.9999 0.0008 -0.0006>}]
 dyn_hair_bk_01_L [{shapeType=Capsule, length=9.7717, radius=1.9882, shapeName=dyn_hair_bk_01_L_capsule, center=<5.0875 0 0>, rotation=<rot -90 0 0>}]
 dyn_hair_bk_01_R [{shapeType=Capsule, length=8.7398, radius=2.5, shapeName=dyn_hair_bk_01_R_capsule, center=<5.0875 0 0>, rotation=<rot -90 -0.0002 0>}]
 dyn_hair_bk_MSTR_L [{shapeType=Capsule, length=0, radius=0, shapeName=dyn_hair_bk_MSTR_L_capsule}]
 dyn_hair_bk_MSTR_R [{shapeType=Capsule, length=0, radius=0, shapeName=dyn_hair_bk_MSTR_R_capsule}]
 dyn_hair_bk_end_01_L [{shapeType=Capsule, length=7.7934, radius=1, shapeName=dyn_hair_bk_end_01_L_capsule, center=<4.6525 0 0>, rotation=<rot -90 0 0>}]
 dyn_hair_bk_end_01_R [{
   shapeType=Capsule
   length=5.3325
   radius=1
   shapeName=dyn_hair_bk_end_01_R_capsule
   center=<5.2239 -0.498 -0.0311>
   rotation=<rot -83.3349 -36.1138 0.0002>
  }]
 dyn_hair_bk_end_02_L [{shapeType=Capsule, length=7.773, radius=1, shapeName=dyn_hair_bk_end_02_L_capsule, center=<5.0875 0 0>, rotation=<rot -90 0 0>}]
 dyn_hair_bk_end_02_R [{shapeType=Capsule, length=4.7923, radius=1, shapeName=dyn_hair_bk_end_02_R_capsule, center=<6.0678 -0.8225 0>, rotation=<rot -90 -15.4379 0>}]
 dyn_hair_bk_end_03_L [{shapeType=Capsule, length=3.394, radius=1, shapeName=dyn_hair_bk_end_03_L_capsule, center=<5.0875 0 0>, rotation=<rot -90 0 0>}]
 dyn_hair_bk_end_03_R [{
   shapeType=Capsule
   length=6.5698
   radius=1
   shapeName=dyn_hair_bk_end_03_R_capsule
   center=<4.4516 0 0.0477>
   rotation=<rot -50.2266 -106.2442 105.3186>
  }]
 dyn_hair_fr_01_L [{shapeType=Capsule, length=7.4067, radius=1.3, shapeName=dyn_hair_fr_01_L_capsule, center=<5.0875 0 0>, rotation=<rot -90 0.0003 0>}]
 dyn_hair_fr_01_R [{shapeType=Capsule, length=5.5069, radius=1.3, shapeName=dyn_hair_fr_01_R_capsule, center=<4.2704 0 0>, rotation=<rot -90 -0.0003 0>}]
 dyn_hair_fr_02_L [{shapeType=Capsule, length=10.3806, radius=1, shapeName=dyn_hair_fr_02_L_capsule, center=<5.0875 0 0>, rotation=<rot -90 0.0003 0>}]
 dyn_hair_fr_02_R [{shapeType=Capsule, length=9.6674, radius=1, shapeName=dyn_hair_fr_02_R_capsule, center=<4.6012 0 0>, rotation=<rot -90 -0.0003 0>}]
 dyn_hair_fr_MSTR_L [{shapeType=Capsule, length=0, radius=0, shapeName=dyn_hair_fr_MSTR_L_capsule}]
 dyn_hair_fr_MSTR_R [{shapeType=Capsule, length=0, radius=0, shapeName=dyn_hair_fr_MSTR_R_capsule}]
 dyn_hair_mid_01_L [{shapeType=Capsule, length=9.2201, radius=1, shapeName=dyn_hair_mid_01_L_capsule, center=<5.0875 0 0>, rotation=<rot -90 0.0003 0>}]
 dyn_hair_mid_01_R [{shapeType=Capsule, length=7.0748, radius=1, shapeName=dyn_hair_mid_01_R_capsule, center=<4.4126 0 0>, rotation=<rot -90 -0.0003 0>}]
 dyn_hair_mid_02_L [{shapeType=Capsule, length=8.1855, radius=1, shapeName=dyn_hair_mid_02_L_capsule, center=<5.0875 0 0>, rotation=<rot -90 0.0003 0>}]
 dyn_hair_mid_02_R [{shapeType=Capsule, length=7.0252, radius=1, shapeName=dyn_hair_mid_02_R_capsule, center=<3.5524 0 0>, rotation=<rot -90 -0.0003 0>}]
 dyn_hair_mid_MSTR_L [{shapeType=Capsule, length=0, radius=0, shapeName=dyn_hair_mid_MSTR_L_capsule}]
 dyn_hair_mid_MSTR_R [{shapeType=Capsule, length=0, radius=0, shapeName=dyn_hair_mid_MSTR_R_capsule}]
 head [
  {shapeType=Capsule, length=3.8998, radius=7.1002, shapeName=head_capsule, center=<5.783 1.0306 0>, rotation=<rot 90 -0.0007 0>}
  {shapeType=Capsule, length=4, radius=6.5, shapeName=head_capsule, center=<-1.1869 3.1217 0>, rotation=<rot 90 -0.0001 0>}
 ]
 neck_01 [{shapeType=Capsule, length=7, radius=5, shapeName=neck_01_capsule, center=<2.2544 0 0>, rotation=<rot 90 -20.0008 0>}]
 spine_05 [{shapeType=Capsule, length=8, radius=11, shapeName=spine_05_capsule}]
 upperarm_l [{shapeType=Capsule, length=24, radius=4, shapeName=upperarm_r_capsule, center=<12.5297 0 0>, rotation=<rot -0.0004 -89.9996 -100>}]
 upperarm_r [{shapeType=Capsule, length=24, radius=4, shapeName=upperarm_r_capsule, center=<-12.5297 0.2933 0>, rotation=<rot -0.0004 -89.9996 -100>}]
}
constraints [
 {
  bone1Name=head
  bone2Name=neck_01
  swing1LimitDegrees=45
  swing1Motion=Free
  swing2LimitDegrees=45
  swing2Motion=Free
  twistLimitDegrees=45
  twistMotion=Free
 }
 {
  bone1Name=neck_01
  bone2Name=spine_05
  swing1LimitDegrees=45
  swing1Motion=Free
  swing2LimitDegrees=45
  swing2Motion=Free
  twistLimitDegrees=45
  twistMotion=Free
 }
 {
  bone1Name=clavicle_l
  bone2Name=spine_05
  swing1LimitDegrees=45
  swing1Motion=Free
  swing2LimitDegrees=45
  swing2Motion=Free
  twistLimitDegrees=45
  twistMotion=Free
 }
 {
  bone1Name=clavicle_r
  bone2Name=spine_05
  swing1LimitDegrees=45
  swing1Motion=Free
  swing2LimitDegrees=45
  swing2Motion=Free
  twistLimitDegrees=45
  twistMotion=Free
 }
 {
  bone1Name=upperarm_l
  bone2Name=clavicle_l
  swing1LimitDegrees=45
  swing1Motion=Free
  swing2LimitDegrees=45
  swing2Motion=Free
  twistLimitDegrees=45
  twistMotion=Free
 }
 {
  bone1Name=upperarm_r
  bone2Name=clavicle_r
  swing1LimitDegrees=45
  swing1Motion=Free
  swing2LimitDegrees=45
  swing2Motion=Free
  twistLimitDegrees=45
  twistMotion=Free
 }
 {bone1Name=dyn_hair_bk_MSTR_L, bone2Name=head, swing1LimitDegrees=10, swing2LimitDegrees=25, twistLimitDegrees=2.5}
 {bone1Name=dyn_hair_bk_01_L, bone2Name=dyn_hair_bk_MSTR_L, swing1LimitDegrees=20, swing2LimitDegrees=30, twistLimitDegrees=2.5}
 {bone1Name=dyn_hair_bk_end_01_L, bone2Name=dyn_hair_bk_01_L, swing1LimitDegrees=20, swing2LimitDegrees=30, twistLimitDegrees=2.5}
 {bone1Name=dyn_hair_bk_end_02_L, bone2Name=dyn_hair_bk_01_L, swing1LimitDegrees=20, swing2LimitDegrees=30, twistLimitDegrees=2.5}
 {bone1Name=dyn_hair_bk_end_03_L, bone2Name=dyn_hair_bk_01_L, swing1LimitDegrees=20, swing2LimitDegrees=30, twistLimitDegrees=2.5}
 {bone1Name=dyn_hair_bk_MSTR_R, bone2Name=head, swing1LimitDegrees=10, swing2LimitDegrees=25, twistLimitDegrees=2.5}
 {bone1Name=dyn_hair_bk_01_R, bone2Name=dyn_hair_bk_MSTR_R, swing1LimitDegrees=25, swing2LimitDegrees=30, twistLimitDegrees=2.5}
 {bone1Name=dyn_hair_bk_end_01_R, bone2Name=dyn_hair_bk_01_R, swing1LimitDegrees=25, swing2LimitDegrees=30, twistLimitDegrees=2.5}
 {bone1Name=dyn_hair_bk_end_02_R, bone2Name=dyn_hair_bk_01_R, swing1LimitDegrees=25, swing2LimitDegrees=30, twistLimitDegrees=2.5}
 {bone1Name=dyn_hair_bk_end_03_R, bone2Name=dyn_hair_bk_01_R, swing1LimitDegrees=25, swing2LimitDegrees=30, twistLimitDegrees=2.5}
 {bone1Name=dyn_hair_fr_MSTR_L, bone2Name=head, swing1LimitDegrees=10, swing2LimitDegrees=15, twistLimitDegrees=2.5}
 {bone1Name=dyn_hair_fr_01_L, bone2Name=dyn_hair_fr_MSTR_L, swing1LimitDegrees=10, swing2LimitDegrees=15, twistLimitDegrees=2.5}
 {bone1Name=dyn_hair_fr_02_L, bone2Name=dyn_hair_fr_01_L, swing1LimitDegrees=15, swing2LimitDegrees=15, twistLimitDegrees=2.5}
 {bone1Name=dyn_hair_fr_MSTR_R, bone2Name=head, swing1LimitDegrees=10, swing2LimitDegrees=15, twistLimitDegrees=2.5}
 {bone1Name=dyn_hair_fr_01_R, bone2Name=dyn_hair_fr_MSTR_R, swing1LimitDegrees=15, swing2LimitDegrees=15, twistLimitDegrees=2.5}
 {bone1Name=dyn_hair_fr_02_R, bone2Name=dyn_hair_fr_01_R, swing1LimitDegrees=15, swing2LimitDegrees=15, twistLimitDegrees=2.5}
 {bone1Name=dyn_hair_mid_MSTR_L, bone2Name=head, swing1LimitDegrees=10, swing2LimitDegrees=15, twistLimitDegrees=2.5}
 {bone1Name=dyn_hair_mid_01_L, bone2Name=dyn_hair_mid_MSTR_L, swing1LimitDegrees=10, swing2LimitDegrees=15, twistLimitDegrees=2.5}
 {bone1Name=dyn_hair_mid_02_L, bone2Name=dyn_hair_mid_01_L, swing1LimitDegrees=15, swing2LimitDegrees=15, twistLimitDegrees=2.5}
 {bone1Name=dyn_hair_mid_MSTR_R, bone2Name=head, swing1LimitDegrees=10, swing2LimitDegrees=15, twistLimitDegrees=2.5}
 {bone1Name=dyn_hair_mid_01_R, bone2Name=dyn_hair_mid_MSTR_R, swing1LimitDegrees=10, swing2LimitDegrees=15, twistLimitDegrees=2.5}
 {bone1Name=dyn_hair_mid_02_R, bone2Name=dyn_hair_mid_01_R, swing1LimitDegrees=15, swing2LimitDegrees=15, twistLimitDegrees=2.5}
 {bone1Name=dyn_hair_01_L, bone2Name=dyn_hair_MSTR_L, swing1LimitDegrees=5, swing2LimitDegrees=0, twistLimitDegrees=0}
 {bone1Name=dyn_hair_02_L, bone2Name=dyn_hair_01_L, swing1LimitDegrees=2.5, swing2LimitDegrees=0, twistLimitDegrees=0}
]
