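# Animation Blueprint for PongoQuell's face accessory, driving secondary motion on a cosmetic part that includes hair bones.
# AnimGraph copies the parent mesh pose, applies world-space upperarm overrides, runs a Control Rig fed crouch and skydive state, then simulates RigidBody physics in head space.
# EventGraph updates gravity and simulation settings through a parameter-asset state function, derives a skydive value, and caches BodyMesh upperarm transforms.
# An additive hair-end rotation chain sits unconnected.
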AnimBlueprint /BRCosmetics/Characters/Player/Female/Medium/Bodies/F_MED_PongoQuell/Meshes/Parts/F_MED_PongoQuell_FaceAcc_AnimBP
target_skeleton=&/F_MED_PongoQuell_FaceAcc_Skeleton
graphs [
 AnimGraph {nodes [
   n1 = "Transform (Modify) Bone - Bone: dyn_hair_bk_end_01_L"(Alpha=1, Translation="0,0,0", Rotation="0,-15,0", Scale="1,1,1") {rotationMode=BMM_Additive, rotationSpace=BCS_BoneSpace, boneToModify {boneName=dyn_hair_bk_end_01_L}, rotation {yaw=-15}}
   n2 = "Transform (Modify) Bone - Bone: dyn_hair_bk_end_02_L"(ComponentPose=@n1, Alpha=1, Translation="0,0,0", Rotation="0,-15,0", Scale="1,1,1") {rotationMode=BMM_Additive, rotationSpace=BCS_BoneSpace, boneToModify {boneName=dyn_hair_bk_end_02_L}, rotation {yaw=-15}}
   n3 = "Transform (Modify) Bone - Bone: dyn_hair_bk_end_03_L"(ComponentPose=@n2, Alpha=1, Translation="0,0,0", Rotation="0,-15,0", Scale="1,1,1") {rotationMode=BMM_Additive, rotationSpace=BCS_BoneSpace, boneToModify {boneName=dyn_hair_bk_end_03_L}, rotation {yaw=-15}}
   n4 = "Transform (Modify) Bone - Bone: dyn_hair_bk_end_01_R"(ComponentPose=@n3, Alpha=1, Translation="0,0,0", Rotation="0,15,0", Scale="1,1,1") {rotationMode=BMM_Additive, rotationSpace=BCS_BoneSpace, boneToModify {boneName=dyn_hair_bk_end_01_R}, rotation {yaw=15}}
   n5 = "Transform (Modify) Bone - Bone: dyn_hair_bk_end_02_R"(ComponentPose=@n4, Alpha=1, Translation="0,0,0", Rotation="0,15,0", Scale="1,1,1") {rotationMode=BMM_Additive, rotationSpace=BCS_BoneSpace, boneToModify {boneName=dyn_hair_bk_end_02_R}, rotation {yaw=15}}
   "Transform (Modify) Bone - Bone: dyn_hair_bk_end_03_R"(ComponentPose=@n5, Alpha=1, Translation="0,0,0", Rotation="0,15,0", Scale="1,1,1") {rotationMode=BMM_Additive, rotationSpace=BCS_BoneSpace, boneToModify {boneName=dyn_hair_bk_end_03_R}, rotation {yaw=15}}
   n6 = "Transform (Modify) Bone - Bone: upperarm_l"!(ComponentPose=Local To Component(LocalPose=Copy Pose From Mesh() {bUseAttachedParent=true}), Alpha=1, Translation="0,0,0", Rotation="0,0,0") -> (Pose) {
    rotationMode=BMM_Replace
    rotationSpace=BCS_WorldSpace
    translationMode=BMM_Replace
    translationSpace=BCS_WorldSpace
    alphaBoolBlend {blendOption=Linear}
    boneToModify {boneName=upperarm_l}
   }
   n7 = "Transform (Modify) Bone - Bone: upperarm_r"!(ComponentPose=@n6, Alpha=1, Translation="0,0,0", Rotation="0,0,0") -> (Pose) {
    rotationMode=BMM_Replace
    rotationSpace=BCS_WorldSpace
    translationMode=BMM_Replace
    translationSpace=BCS_WorldSpace
    alphaBoolBlend {blendOption=Linear}
    boneToModify {boneName=upperarm_r}
   }
   n8 = Control Rig!(
    Source=Component To Local(ComponentPose=@n7)
    Alpha
    is crouching:bool=Get bIsCrouching()
    Is Sky Diving:double=Get IsSky Diving() -> (IsSky Diving:double)
   ) -> (Pose) {controlRigAssetReference {controlRigAsset=&/F_MED_PongoQuell_FaceAcc_ControlRig}}
   n9 = RigidBody!(
    ComponentPose=Local To Component(LocalPose=@n8)
    Alpha
    OverrideWorldGravity:Vector="0,0,-980"
    SimSpaceSettings:Sim Space Settings Structure=(WorldAlpha=1,VelocityScaleZ=1,DampingAlpha=1,MaxLinearVelocity=10000,MaxAngularVelocity=10000,MaxLinearAcceleration=10000,MaxAngularAcceleration=10000,ExternalLinearDragV=(X=0,Y=0,Z=0),ExternalLinearVelocity=(X=0,Y=0,Z=0),ExternalAngularVelocity=(X=0,Y=0,Z=0))
   ) -> (Pose) {
    bClampLinearTranslationLimitToRefPose=true
    bOverrideWorldGravity=true
    overridePhysicsAsset=&/F_MED_PongoQuell_FaceAcc_Physics
    simulationSpace=BaseBoneSpace
    baseBoneRef {boneName=head}
    simSpaceSettings {worldAlpha=1}
   }
   Output Pose(Result=Component To Local(ComponentPose=@n9)) {layerGroup=None}
  ]}
 EventGraph {nodes [
   Event BlueprintUpdateAnimation()
   > n1 = Sequence()
   > n2 = "Gravity Override State Function (Parameter Asset) (Nativized)"(
    CharacterPartAnimInstance=Self-Reference()
    GameTimeInSeconds=GetGameTimeInSeconds()
    Parameters=&/F_MED_PongoQuell_FaceAcc_Parameters
   )
   > Set Gravity Override(Gravity Override=@n2)
   > Set RBAN Settings Out(RBAN Settings Out=@n2.OutRigidBodyAnimNodeParameters)
   n3 = Get bIsParachuteOpened()
   > Set IsSky Diving(IsSky Diving:double=SelectFloat(A=1, B=SelectFloat(A=0.5, B=0, bPickA=@n3), bPickA=Get bIsSkydiveFloating())) -> (Output_Get:double)
   n4 = Get BodyMesh() -> (BodyMesh:Skeletal Mesh Component Object Reference)
   Is Valid(exec=@n1.then_1, InputObject=@n4)
   n5 = BreakTransform(InTransform=GetSocketTransform(self=@n4, InSocketName=upperarm_l, TransformSpace=RTS_World))
   > Set upperarm_l translation(upperarm_l translation:Vector=@n5) -> (Output_Get:Vector)
   > Set upperarm_l rotation(upperarm_l rotation:Rotator=@n5.Rotation) -> (Output_Get:Rotator)
   n6 = BreakTransform(InTransform=GetSocketTransform(self=@n4, InSocketName=upperarm_r, TransformSpace=RTS_World))
   > Set upperarm_r translation(upperarm_r translation:Vector=@n6) -> (Output_Get:Vector)
   > Set upperarm_r rotation(upperarm_r rotation:Rotator=@n6.Rotation) -> (Output_Get:Rotator)
   n7 = Event OnCharacterCustomizationLoaded()
   > Set BodyMesh(BodyMesh=Get Array Item with EFort Custom Part Type(Array=@n7.CustomMeshParts))
   PrintString!(execute, InString="To String (Name)"(InName=@n2.OutCurrentState), bPrintToScreen, bPrintToLog, TextColor, Duration, Key) -> (then)
  ]}
]
variables [
 BodyMesh {type=SkeletalMeshComponent}
 Gravity Override {type=Vector}
 IsSky Diving {type=number}
 RBAN Settings Out {type=FortRigidBodyAnimNodeParameters}
 upperarm_l rotation {type=Rotator}
 upperarm_l translation {type=Vector}
 upperarm_r rotation {type=Rotator}
 upperarm_r translation {type=Vector}
]
